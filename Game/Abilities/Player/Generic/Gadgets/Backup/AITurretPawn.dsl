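; AI sentry turret pawn for the Backup floor trap, which sets its team, health, shield, range, accuracy and scoring.
; On authority, sphere overlaps maintain targets, with a pausable rescan timer for missed cases like vehicle entry.
; A timer randomizes aim offset and sends the shoot ability event; adaptive aim tightens spread while the target is visible.
; Another timer moves an aim-at child actor and AI focus; cosmetic events handle meshes, reload, headshot and teleport effects.
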(event FortAIPawn|EventOnTeamChanged
  ;[1 1 1] Clear Current Targets and check current overlapping actors for new targets
  (CallFunction|ReEvaluateTargets))

(event EventBeginPlay
  (Variables|HUD|SetbHideHealthBar true)
  (Networking|SwitchHasAuthority
    (:Authority
      (CallFunction|BindFindTargetTimer)
      (Variables|Configurables|SetCurrentAccuracyRadius (Variables|Configurables|GetMaxAccuracyRadius))
      (bind _returnvalue (Utilities|Time|SetTimerbyEvent (EventDispatchers|CreateEvent :!event EvaluateTarget) (Variables|Configurables|GetTargetEvaluationInterval) true))
      (Variables|Default|SetEvaluationTimerHandle _returnvalue))
    (:Remote
      _))
  (bind _returnvalue_2 (Utilities|Time|SetTimerbyEvent (EventDispatchers|CreateEvent :!event UpdateAimTarget) (Variables|Configurables|GetRotationUpdateInterval) true))
  (Variables|Default|SetUpdateRotationTimerHandle _returnvalue_2)
  (bind _asathena_aicontroller (Utilities|Casting|CastToAthenaAIController (AI|GetAIController self)))
  (Utilities|IsValid (AI|Perception|GetAIPerceptionComponent _asathena_aicontroller)
    (:Is%20Valid
      (AI|Perception|SetSenseEnabled (AI|Perception|GetAIPerceptionComponent _asathena_aicontroller) "/Script/FortniteAI.AthenaAISense_Hearing" false))))

(event Custom|ReloadTimer_ShotgunFX
  (params (Time :type Float%20%28double-precision%29))
  (Utilities|Time|SetTimerbyEvent (delegate Custom|EndShotgun_ReloadLoop) Time))

(event EventEndPlay
  (params (EndPlayReason :type EEndPlayReason%20Enum :default "Destroyed"))
  (Utilities|Time|ClearandInvalidateTimerbyHandle (Variables|Default|GetEvaluationTimerHandle))
  (Utilities|Time|ClearandInvalidateTimerbyHandle (Variables|Default|GetUpdateRotationTimerHandle)))

(event Custom|SetupSkeletalMesh
  ;[1 1 1] Setup The AI Skeletal Mesh
  (bind _object (Utilities|AsyncLoadAsset (Variables|Default|GetDesiredSkeletalMesh))
    (:Completed
      (Networking|CanExecuteCosmeticEvents
        (:True
          (bind _asskeletal_mesh (Utilities|Casting|CastToSkeletalMesh _object))
          (Components|SkinnedMesh|SetSkinnedAssetandUpdate (Variables|Character|GetMesh) _asskeletal_mesh true))))))

(event Custom|SetupSkeletalHead
  ;[1 1 1] Setup The AI Skeletal Mesh Head
  (bind _object (Utilities|AsyncLoadAsset (Variables|Default|GetDesiredSkeletalHead))
    (:Completed
      (Networking|CanExecuteCosmeticEvents
        (:True
          (bind _asskeletal_mesh (Utilities|Casting|CastToSkeletalMesh _object))
          (Components|SkinnedMesh|SetSkinnedAssetandUpdate (Variables|Default|GetHead) _asskeletal_mesh true))))))

(event Custom|EndShotgun_ReloadLoop
  (Animation|PlayLocalAnimMontage (CharacterAnimation|GetReloadAnimation (Variables|Weapon|GetCurrentWeapon)) 1.0 "Outro"))

(event GameplayCue|GameplayCue.Abilities.Generic.BotTurret.Reload
  (params (EventType :type EGameplayCueEvent%20Enum :default "OnActive") (Parameters :type Gameplay%20Cue%20Parameters%20Structure%20%28by%20ref%29))
  (CallFunction|PlayReload))

(event EventOnDamagePlayEffects
  (params (Damage :type Float%20%28single-precision%29) (DamageTags :type Gameplay%20Tag%20Container%20Structure%20%28by%20ref%29) (Momentum :type Vector :default "0, 0, 0") (HitInfo :type Hit%20Result%20Structure%20%28by%20ref%29) (InstigatedBy :type Fort%20Pawn%20Object%20Reference) (DamageCauser :type Actor%20Object%20Reference) (EffectContext :type Gameplay%20Effect%20Context%20Handle%20Structure))
  (bind _mesh (Variables|Character|GetMesh))
  (bind _returnvalue (Transformation|GetSocketLocation _mesh "FX_Head"))
  ;[1 1 1] Play Damage VFX
  (switch Utilities|FlowControl|Switch|SwitchonEFortDamageZone (DamageZones|GetDamageZone HitInfo)
    (:Critical
      (Effects|Components|ParticleSystem|SpawnFXAtLocation "/Game/Characters/Enemies/Husk/Effects/NS_Headshot.NS_Headshot" _returnvalue (Transformation|GetWorldRotation _mesh))
      (Audio|PlaySoundatLocation "/Game/Sounds/Creative/Gadgets/Traps/Trap_AI_Pain_Cue.Trap_AI_Pain_Cue" _returnvalue))))

(event EventOnDeathPlayEffects
  (params (Damage :type Float%20%28single-precision%29) (DamageTags :type Gameplay%20Tag%20Container%20Structure%20%28by%20ref%29) (Momentum :type Vector :default "0, 0, 0") (HitInfo :type Hit%20Result%20Structure%20%28by%20ref%29) (InstigatedBy :type Fort%20Pawn%20Object%20Reference) (DamageCauser :type Actor%20Object%20Reference) (EffectContext :type Gameplay%20Effect%20Context%20Handle%20Structure))
  (bind _mesh (Variables|Character|GetMesh))
  (bind _returnvalue (Transformation|GetSocketLocation _mesh "FX_Feet"))
  ;[1 1 1] Death VFX
  (Effects|SpawnEmitteratLocation "/Game/Effects/Fort_Effects/Effects/Characters/Generic/ResIn/P_Player_Teleport_In_Athena.P_Player_Teleport_In_Athena" _returnvalue)
  (Effects|SpawnEmitteratLocation "/Game/Athena/Items/Consumables/Telenade/P_Telenade_teleport_temp.P_Telenade_teleport_temp" _returnvalue)
  (Audio|PlaySoundatLocation "/Game/Sounds/Creative/Gadgets/Traps/Trap_AI_Teleport_Out_Cue.Trap_AI_Teleport_Out_Cue" _returnvalue)
  (FortPawn|SetPawnVisibilityLocal true))

(event Custom|SpawnVFX
  (bind _mesh (Variables|Character|GetMesh))
  (bind _returnvalue (Transformation|GetSocketLocation _mesh "FX_Feet"))
  ;[1 1 1] Death VFX
  (Effects|SpawnEmitteratLocation "/Game/Effects/Fort_Effects/Effects/Characters/Generic/ResIn/P_Player_Teleport_In_Athena.P_Player_Teleport_In_Athena" _returnvalue)
  (Effects|SpawnEmitteratLocation "/Game/Athena/Items/Consumables/Telenade/P_Telenade_teleport_temp.P_Telenade_teleport_temp" _returnvalue)
  (Audio|PlaySoundatLocation "/Game/Sounds/Creative/Gadgets/Traps/Trap_AI_Teleport_In_Cue.Trap_AI_Teleport_In_Cue" _returnvalue))

(event EventOnDamageServer
  (params (Damage :type Float%20%28single-precision%29) (DamageTags :type Gameplay%20Tag%20Container%20Structure%20%28by%20ref%29) (Momentum :type Vector :default "0, 0, 0") (HitInfo :type Hit%20Result%20Structure%20%28by%20ref%29) (InstigatedBy :type Controller%20Object%20Reference) (DamageCauser :type Actor%20Object%20Reference) (EffectContext :type Gameplay%20Effect%20Context%20Handle%20Structure))
  (bind _returnvalue (Pawn|GetControlledPawn InstigatedBy))
  (bind _returnvalue_2 (Transformation|GetActorLocation _returnvalue))
  ;[1 1 1] Receive Damage Response
  (Variables|Default|GetCurrentTarget
    (:then)
    (:else
      (if (Utilities|IsValid InstigatedBy :!class /Script/Engine.KismetSystemLibrary)
        (if (Utilities|IsValid _returnvalue :!class /Script/Engine.KismetSystemLibrary)
          (bind _aimatactor (Variables|Default|GetAimAtActor)
            (:then
              (bind _childactor (Variables|/Script/Engine.ChildActorComponent|GetChildActor _aimatactor)
                (:then
                  (Transformation|SetActorLocation _childactor _returnvalue_2 false false)
                  (Utilities|SetGoalActor _childactor
                    (:Then
                      (Utilities|IsValid (Pawn|GetController)
                        (:Is%20Valid
                          (AI|SetAIFocalPoint 0 _returnvalue_2))))))
                (:else)))
            (:else)))))))

(event Custom|AttemptAlterOverlap
  (params (Object :type Actor%20Object%20Reference))
  ;[1 1 1] Overlap Events
  ;;[1 1 1] This can be called as a overlap but from another event, such as FindTargetTimer
  (CallFunction|ActorOverlap Object))

(event OnComponentBeginOverlap(PawnTargetSphere)
  (params (OverlappedComponent :type Primitive%20Component%20Object%20Reference) (OtherActor :type Actor%20Object%20Reference) (OtherComp :type Primitive%20Component%20Object%20Reference) (OtherBodyIndex :type Integer) (bFromSweep :type Boolean) (SweepResult :type Hit%20Result%20Structure%20%28by%20ref%29))
  ;[1 1 1] Overlap Events
  (CallFunction|ActorOverlap OtherActor))

(event OnComponentEndOverlap(PawnTargetSphere)
  (params (OverlappedComponent :type Primitive%20Component%20Object%20Reference) (OtherActor :type Actor%20Object%20Reference) (OtherComp :type Primitive%20Component%20Object%20Reference) (OtherBodyIndex :type Integer))
  ;[1 1 1] Overlap Events
  (Networking|SwitchHasAuthority
    (:Authority
      (bind _asfort_pawn (Utilities|Casting|CastToFortPawn OtherActor))
      (Utilities|Array|RemoveItem (Variables|Default|GetPlayerTargets) _asfort_pawn)
      (bind _asfort_player_pawn (Utilities|Casting|CastToFortPlayerPawn OtherActor)
        (:then
          (EventDispatchers|UnbindallEventsfromOnNewPlayerState _asfort_player_pawn))
        (:CastFailed
          _))
      (if (Utilities|Equal(Object) (Variables|Default|GetCurrentTarget) OtherActor)
        (Variables|Default|SetCurrentTarget 0)
        (Utilities|Time|UnpauseTimerbyHandle (Variables|Default|GetFindTargetTimerHandle))))))

(event Custom|BindFindTargetTimer
  ;[1 1 1] Handles targetting that is not directly handled by overlaps, such as when a player gets into a mech vehicle, when that mech vehicle never triggered a overlap
  (bind _returnvalue (Utilities|Time|SetTimerbyEvent (delegate Custom|FindTargetTimer) 1.0 true))
  (bind _output_get (Variables|Default|SetFindTargetTimerHandle _returnvalue))
  (Utilities|Time|PauseTimerbyHandle _output_get))

(event EventOnDeathServer
  (params (Damage :type Float%20%28single-precision%29) (DamageTags :type Gameplay%20Tag%20Container%20Structure%20%28by%20ref%29) (Momentum :type Vector :default "0, 0, 0") (HitInfo :type Hit%20Result%20Structure%20%28by%20ref%29) (InstigatedBy :type Controller%20Object%20Reference) (DamageCauser :type Actor%20Object%20Reference) (EffectContext :type Gameplay%20Effect%20Context%20Handle%20Structure))
  ;[1 1 1] "Disabling" the sentry on death
  (Utilities|Time|ClearandInvalidateTimerbyHandle (Variables|Default|GetEvaluationTimerHandle))
  (Collision|SetCollisionEnabled (Variables|Default|GetPawnTargetSphere) "NoCollision"))

(event Custom|FindTargetTimer
  ;[1 1 1] Handles targetting that is not directly handled by overlaps, such as when a player gets into a mech vehicle, when that mech vehicle never triggered a overlap
  (for _array_element (Collision|GetOverlappingActors (Variables|Default|GetPawnTargetSphere) "/Script/FortniteGame.FortPlayerPawn" :!class /Script/Engine.PrimitiveComponent)
    (CallFunction|AttemptAlterOverlap _array_element)))

(event OnComponentBeginOverlap(CapsuleComponent)
  (params (OverlappedComponent :type Primitive%20Component%20Object%20Reference) (OtherActor :type Actor%20Object%20Reference) (OtherComp :type Primitive%20Component%20Object%20Reference) (OtherBodyIndex :type Integer) (bFromSweep :type Boolean) (SweepResult :type Hit%20Result%20Structure%20%28by%20ref%29)))

(fn ActorOverlap
  (params (Actor :type Actor%20Object%20Reference))
  (Networking|SwitchHasAuthority
    (:Authority
      (bind _asfort_pawn (Utilities|Casting|CastToFortPawn Actor))
      (CallFunction|PawnOverlap _asfort_pawn))))

(fn AddTarget
  (params (Pawn :type Fort%20Pawn%20Object%20Reference))
  (bind _output_get (LocalVariables|SetNewPawn Pawn))
  (bind _bcanadd (CallFunction|CanAddTarget _output_get))
  (if _bcanadd
    (Utilities|Array|AddUnique (Variables|Default|GetPlayerTargets) (LocalVariables|GetNewPawn))
    (Utilities|Time|PauseTimerbyHandle (Variables|Default|GetFindTargetTimerHandle))))

(fn AdjustAccuracySettings
  (switch int (Variables|/Game/Items/Traps/Blueprints/Trap_Floor_Backup.Trap_Floor_Backup_C|GetAccuracyPresetIndex (Variables|Default|GetParentTrap))
    (:0
      ;[1 1 1] Low
      (Variables|Configurables|SetMaxAccuracyRadius 1.5e+02)
      (Variables|Configurables|SetMinAccuracyRadius 5e+01)
      (Variables|Configurables|SetAccuracyRadiusRate 1e+01))
    (:1
      ;[1 1 1] Medium
      (Variables|Configurables|SetMaxAccuracyRadius 85.0)
      (Variables|Configurables|SetMinAccuracyRadius 25.0)
      (Variables|Configurables|SetAccuracyRadiusRate 1e+01))
    (:2
      ;[1 1 1] High
      (Variables|Configurables|SetMaxAccuracyRadius 75.0)
      (Variables|Configurables|SetMinAccuracyRadius 1e+01)
      (Variables|Configurables|SetAccuracyRadiusRate 1e+01))
    (:3
      ;[1 1 1] Deadshot
      (Variables|Configurables|SetMaxAccuracyRadius 5e+01)
      (Variables|Configurables|SetMinAccuracyRadius 0.0)
      (Variables|Configurables|SetAccuracyRadiusRate 1e+01)))
  (Variables|Configurables|SetCurrentAccuracyRadius (Variables|Configurables|GetMaxAccuracyRadius)))

(fn CanAddTarget
  (params (Pawn :type Fort%20Pawn%20Object%20Reference))
  (bind _self self)
  (bind _returnvalue (Team|OnSameTeam _self Pawn))
  (bind _returnvalue_2 (not _returnvalue))
  (Utilities|Casting|CastToFortAIPawn Pawn
    (:then
      (return (and (Variables|Configurables|GetbCanTargetCreatures) _returnvalue_2)))
    (:CastFailed
      (return _returnvalue_2))) ; Base class for all Wildlife and Deimos Fiends
)

(fn EvaluateTarget
  (if (> (Utilities|Array|Length (Variables|Default|GetPlayerTargets)) 0)
    (CallFunction|UpdateAimOffset)
    (Ability|SendGameplayEventtoActor self "(TagName=\"Abilities.Generic.BotTurret.Shoot\")")))

(fn GetAimTargetLocation
  (params (AimTarget :type Fort%20Pawn%20Object%20Reference) (TargetSocketName :type Name :default "FX_Chest"))
  (bind _targetsocketname (Variables|GetTargetSocketName))
  (bind _aimtarget (Variables|GetAimTarget))
  (bind _mesh (Variables|/Script/Engine.Character|GetMesh _aimtarget))
  (bind _returnvalue (Transformation|GetSocketLocation _mesh _targetsocketname))
  (LocalVariables|SetActorLocation (Transformation|GetActorLocation AimTarget))
  (LocalVariables|SetbHasTargetSocket (Transformation|DoesSocketExist (Variables|/Script/Engine.Character|GetMesh AimTarget) (Variables|GetTargetSocketName)))
  (Utilities|FlowControl|Sequence
    (:then_0
      (Utilities|Casting|CastToFortPlayerPawn (Variables|GetAimTarget)
        (:then
          (if (LocalVariables|GetbHasTargetSocket)
            (return (Transformation|GetSocketLocation (Variables|/Script/Engine.Character|GetMesh (Variables|GetAimTarget)) (Variables|GetTargetSocketName)))))
        (:CastFailed
          (bind _asfort_aipawn (Utilities|Casting|CastToFortAIPawn (Variables|GetAimTarget)))
          (if (LocalVariables|GetbHasTargetSocket)
            (if (Variables|/Script/FortniteGame.FortAIPawn|GetbUseSharedAnimation _asfort_aipawn)
              ;[1 0.9349 0.5705] Animation Sharing causes mesh and socket locations to be returned in local space instead of world space
              (return (Math|Transform|TransformLocation (Transformation|GetWorldTransform (Variables|/Script/Engine.Actor|GetRootComponent _aimtarget)) _returnvalue))
              (else
                (return _returnvalue)))))))
    (:then_1
      ;[0.7917 0.5327 0.5012] Fallback case
      (return (LocalVariables|GetActorLocation)))))

(fn GetPlayerStateFromPawn
  (params (Pawn :type Fort%20Player%20Pawn%20Object%20Reference))
  (bind _playerstate (Variables|/Script/Engine.Pawn|GetPlayerState Pawn))
  (Utilities|IsValid _playerstate
    (:Is%20Valid
      (return _playerstate))
    (:Is%20Not%20Valid
      (bind _asfort_character_vehicle (Utilities|Casting|CastToFortCharacterVehicle (Utilities|ToObject(Interface) (Vehicle|GetVehicleInterface Pawn)))
        (:then
          (return (Variables|/Script/Engine.Pawn|GetPlayerState _asfort_character_vehicle)))
        (:CastFailed
          (return 0))))))

(fn IsInsideVehicle
  (params (Pawn :type Fort%20Player%20Pawn%20Object%20Reference))
  (return (Utilities|IsValid (Utilities|ToObject(Interface) (Vehicle|GetVehicleInterface Pawn)) :!class /Script/Engine.KismetSystemLibrary)))

(fn IsTargetVisible
  (bind _outhit (Collision|LineTraceByChannel (+ (Transformation|GetSocketLocation (Variables|Character|GetMesh) "FX_Head") (Transformation|GetActorLocation)) (CallFunction|GetAimTargetLocation (Variables|Default|GetCurrentTarget)) "TraceTypeQuery1" true))
  (if (Utilities|Equal(Object) (.HitActor (Collision|BreakHitResult _outhit)) (Variables|Default|GetCurrentTarget))
    (return true)
    (else
      (return false))))

(fn IsValidVehicleTarget
  (params (Pawn :type Fort%20Player%20Pawn%20Object%20Reference))
  (return (> (Utilities|Array|Length (Vehicle|GetAllPassengers (Vehicle|GetVehicleInterface Pawn))) 0)))

(fn OnNewPlayerState
  (params (AffectedPawn :type Fort%20Player%20Pawn%20Object%20Reference))
  (CallFunction|PawnOverlap AffectedPawn))

(fn OnRep_DesiredSkeletalHead
  (CallFunction|SetupSkeletalHead))

(fn OnRep_DesiredSkeletalMesh
  (CallFunction|SetupSkeletalMesh))

(fn PawnOverlap
  (params (Pawn :type Fort%20Pawn%20Object%20Reference))
  (bind _output_get (LocalVariables|SetOverlappingPawn Pawn))
  (Utilities|IsValid _output_get
    (:Is%20Valid
      (bind _asfort_player_pawn (Utilities|Casting|CastToFortPlayerPawn (LocalVariables|GetOverlappingPawn))
        (:then
          (bind _playerstate (CallFunction|GetPlayerStatefromPawn _asfort_player_pawn))
          (Utilities|IsValid _playerstate
            (:Is%20Valid
              (EventDispatchers|UnbindallEventsfromOnNewPlayerState _asfort_player_pawn)
              (bind _vehicle (CallFunction|IsInsideVehicle _asfort_player_pawn))
              (if _vehicle
                (bind _valid (CallFunction|IsValidVehicleTarget _asfort_player_pawn))
                (if _valid
                  (CallFunction|AddTarget _asfort_player_pawn))
                (else
                  (CallFunction|AddTarget _asfort_player_pawn))))
            (:Is%20Not%20Valid
              (EventDispatchers|BindEventtoOnNewPlayerState _asfort_player_pawn (EventDispatchers|CreateEvent :!event OnNewPlayerState)))))
        (:CastFailed
          (CallFunction|AddTarget (LocalVariables|GetOverlappingPawn)))))))

(fn PlayReload
  (Utilities|IsValid (Variables|Weapon|GetCurrentWeapon)
    (:Is%20Valid
      (bind _returnvalue (Animation|PlayLocalAnimMontage (CharacterAnimation|GetReloadAnimation (Variables|Weapon|GetCurrentWeapon))))
      (if (Utilities|Equal(Object) (CharacterAnimation|GetReloadAnimation (Variables|Weapon|GetCurrentWeapon)) "/Game/Animation/Game/MainPlayer/Combat/Guns/Shotgun/Shotgun_SG_BFPAReload_Montage.Shotgun_SG_BFPAReload_Montage")
        (CallFunction|ReloadTimerShotgunFX _returnvalue)))))

(fn ReEvaluateTargets
  (Utilities|Array|Clear (Variables|Default|GetPlayerTargets))
  (for _array_element (Collision|GetOverlappingActors (Variables|Default|GetPawnTargetSphere) "/Script/FortniteGame.FortPawn" :!class /Script/Engine.PrimitiveComponent)
    (CallFunction|ActorOverlap _array_element)))

(fn SetParentTrap
  (params (Parent%20Trap :type Trap%20Floor%20Backup%20Object%20Reference))
  (bind _parenttrap (Variables|Default|GetParentTrap))
  (bind _healthpresetindex (Variables|/Game/Items/Traps/Blueprints/Trap_Floor_Backup.Trap_Floor_Backup_C|GetHealthPresetIndex _parenttrap))
  (bind _healthpresets (Variables|/Game/Items/Traps/Blueprints/Trap_Floor_Backup.Trap_Floor_Backup_C|GetHealthPresets _parenttrap))
  (bind _output (Utilities|Array|Get(aref) _healthpresets _healthpresetindex))
  (bind _shieldpresetindex (Variables|/Game/Items/Traps/Blueprints/Trap_Floor_Backup.Trap_Floor_Backup_C|GetShieldPresetIndex _parenttrap))
  (bind _shieldpresets (Variables|/Game/Items/Traps/Blueprints/Trap_Floor_Backup.Trap_Floor_Backup_C|GetShieldPresets _parenttrap))
  (bind _output_2 (Utilities|Array|Get(aref) _shieldpresets _shieldpresetindex))
  (Utilities|IsValid Parent%20Trap
    (:Is%20Valid
      (bind _output_get (Variables|Default|SetParentTrap Parent%20Trap))
      (Class|TrapFloorBackup|BindDeathDelegate _output_get)
      (CallFunction|AdjustAccuracySettings)
      (if (!= _output -1.0)
        (Fortnite|SetCanbeDamaged 0 true)
        (FortPawn|SetMaxHealth _output)
        (FortPawn|SetHealth _output)
        (else
          (Fortnite|SetCanbeDamaged 0)))
      (FortPawn|SetMaxShield _output_2)
      (FortPawn|SetShield _output_2)
      (Variables|Configurables|SetbCanTargetCreatures (Variables|/Game/Items/Traps/Blueprints/Trap_Floor_Backup.Trap_Floor_Backup_C|GetbCanTargetCreatures (Variables|Default|GetParentTrap)))
      (FortAIPawn|SetOverridenScore (Math|Integer|SelectInt 17 (Variables|/Game/Items/Traps/Blueprints/Trap_Floor_Backup.Trap_Floor_Backup_C|GetOverrideEliminationScore (Variables|Default|GetParentTrap)) (== (Variables|/Game/Items/Traps/Blueprints/Trap_Floor_Backup.Trap_Floor_Backup_C|GetOverrideEliminationScore (Variables|Default|GetParentTrap)) -1)))
      (if (not (Variables|/Game/Items/Traps/Blueprints/Trap_Floor_Backup.Trap_Floor_Backup_C|GetbAwardElimination (Variables|Default|GetParentTrap)))
        (Tags|AddNonReplicatedTagtoPawn self (GameplayTags|MakeLiteralGameplayTag "(TagName=\"Stat.IgnoreKillScore\")")))
      (Components|Sphere|SetSphereRadius (Variables|Default|GetPawnTargetSphere) (Utilities|Array|Get(acopy) (Variables|/Game/Items/Traps/Blueprints/Trap_Floor_Backup.Trap_Floor_Backup_C|GetRangePresets (Variables|Default|GetParentTrap)) (Variables|/Game/Items/Traps/Blueprints/Trap_Floor_Backup.Trap_Floor_Backup_C|GetRangePresetIndex (Variables|Default|GetParentTrap))) true)
      (Variables|Configurables|SetbUseLosTargetting (== (Variables|/Game/Items/Traps/Blueprints/Trap_Floor_Backup.Trap_Floor_Backup_C|GetTargetPresetIndex (Variables|Default|GetParentTrap)) 1))
      (Variables|Configurables|SetbUseAdaptiveAim (== (Variables|/Game/Items/Traps/Blueprints/Trap_Floor_Backup.Trap_Floor_Backup_C|GetAdaptivePresetIndex (Variables|Default|GetParentTrap)) 1))
      (bind _asfort_aicontroller (Utilities|Casting|CastToFortAIController (Pawn|GetController)))
      (AI|SetTeamInt _asfort_aicontroller (Class|TrapFloorBackup|GetTeamIndex (Variables|Default|GetParentTrap)))
      (Collision|SetCollisionEnabled (Variables|Default|GetPawnTargetSphere) "QueryOnly"))))

(fn ShouldUpdateSkeletalMesh
  (return true))

(fn ShouldUseFiringCapsuleCollision
  (return true))

(fn UpdateAimOffset
  (Utilities|IsValid (Variables|Default|GetCurrentTarget)
    (:Is%20Valid
      (Variables|Default|SetTargetAimOffset (+ (* (Transformation|GetActorUpVector (Variables|Default|GetCurrentTarget)) (Math|Random|RandomFloatinRange (* -1 (Variables|Configurables|GetCurrentAccuracyRadius)) (Variables|Configurables|GetCurrentAccuracyRadius))) (* (Transformation|GetActorRightVector (Variables|Default|GetCurrentTarget)) (Math|Random|RandomFloatinRange (* -1 (Variables|Configurables|GetCurrentAccuracyRadius)) (Variables|Configurables|GetCurrentAccuracyRadius)))))
      (if (Variables|Configurables|GetbUseAdaptiveAim)
        (bind _visible (CallFunction|IsTargetVisible))
        (if _visible
          (Variables|Configurables|SetCurrentAccuracyRadius (Math|Float|Clamp(Float) (- (Variables|Configurables|GetCurrentAccuracyRadius) (Variables|Configurables|GetAccuracyRadiusRate)) (Variables|Configurables|GetMinAccuracyRadius) (Variables|Configurables|GetMaxAccuracyRadius)))
          (else
            (Variables|Configurables|SetCurrentAccuracyRadius (Math|Float|Clamp(Float) (+ (Variables|Configurables|GetCurrentAccuracyRadius) (Variables|Configurables|GetAccuracyRadiusRate)) (Variables|Configurables|GetMinAccuracyRadius) (Variables|Configurables|GetMaxAccuracyRadius)))))))))

(fn UpdateAimTarget
  (bind _currenttarget (Variables|Default|GetCurrentTarget))
  (bind _visibilitylocation (CallFunction|GetAimTargetLocation _currenttarget))
  (bind _targetaimoffset (Variables|Default|GetTargetAimOffset))
  (bind _returnvalue (+ _visibilitylocation _targetaimoffset))
  (Variables|Default|GetCurrentTarget
    (:then
      (bind _aimatactor (Variables|Default|GetAimAtActor)
        (:then
          (bind _childactor (Variables|/Script/Engine.ChildActorComponent|GetChildActor _aimatactor)
            (:then
              (Transformation|SetActorLocation _childactor _returnvalue false false)
              (Utilities|SetGoalActor _childactor
                (:Then
                  (Goals|SetCurrentAimTarget (Variables|Default|GetCurrentTarget))
                  (Utilities|IsValid (Pawn|GetController)
                    (:Is%20Valid
                      (AI|SetAIFocalPoint (Variables|Default|GetCurrentTarget) _returnvalue))))))
            (:else)))
        (:else)))
    (:else
      (if (or (== (Utilities|Array|Length (Variables|Default|GetPlayerTargets)) 0) (not (Variables|Default|GetAlreadyClearedTarget)))
        (Utilities|SetGoalActor 0
          (:Then
            (Goals|SetCurrentAimTarget 0)
            (Variables|Default|SetAlreadyClearedTarget true)
            (Utilities|IsValid (Pawn|GetController)
              (:Is%20Valid
                (AI|SetAIFocalPoint 0 (+ (* (Transformation|GetActorForwardVector) 5e+02) (Transformation|GetActorLocation)))))))))))

(fn ConstructionScript)
